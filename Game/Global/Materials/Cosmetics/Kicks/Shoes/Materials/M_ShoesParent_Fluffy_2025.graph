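# Parent material for fluffy shoe cosmetics on skeletal meshes, built on material attributes with the shading model chosen in the graph.
# The character parent PreFX function feeds MF_FluffyFabric, whose tech art mask is gated by a step on the U texture coordinate.
# MF_Sock then processes the result using its opacity mask and the original tech art mask, before the character parent PostFX function finishes it.
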Material /Game/Global/Materials/Cosmetics/Kicks/Shoes/Materials/M_ShoesParent_Fluffy_2025
graph {
 properties {
  bAutomaticallySetUsageInEditor=false
  bUseMaterialAttributes=true
  bUsedWithClothing=true
  bUsedWithMeshDeformer=true
  bUsedWithMorphTargets=true
  bUsedWithSkeletalMesh=true
  pixelDepthOffsetMode=PDOM_Legacy
  shadingModel=MSM_FromMaterialExpression
  usedShadingModels=Default Lit | Subsurface | Clear Coat | Subsurface Profile | Cloth
 }
 nodes [
  n1 = MaterialFunctionCall() {materialFunction=&MF_CharacterParent_2023_PreFX}
  n2 = ComponentMask(None=TextureCoordinate()) {r=true}
  n3 = MaterialFunctionCall!(Material=@n1.Result, TechArtMask=Multiply(A=@n1.Tech Art Mask, B=Step(Y=@n2))) {materialFunction=/Game/Global/Materials/Character_Parent_2023/FluffyFabric_Parent_2023/MF_FluffyFabric}
  MaterialFunctionCall(
   Material=MaterialFunctionCall!(
    InOpacity=GetMaterialAttributes(None=@n3.Result) {attributeGetTypes [#1]}.Opacity Mask
    Material=@n3.Result
    Mask=@n1.Tech Art Mask
   ) {materialFunction=/Game/Global/Materials/Cosmetics/Kicks/Socks/Materials/MF_Sock}.Result
   M=@n1.M
   Pre Skinned Local Position=@n1.M
   Distance Blend=@n1.M
   Fresnel=@n1.M
  ) {materialFunction=&MF_CharacterParent_2023_PostFX}
 ]
}
