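# Test material function for shoe cosmetics that makes masked areas react to gameplay and audio, behind a static switch.
# Reactivity strength is the maximum of powered audio amplitude, a fade over four seconds after an elimination time, and a velocity parameter.
# Two selectable mask channels with their own colors build the emissive.
# Material attribute blends combine the input material, a metallic test look and a black-based emissive look.
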MaterialFunction /Game/Global/Materials/Cosmetics/Kicks/Shoes/Materials/MF_ShoeReactivityTest
graph {nodes [
  n1 = NamedRerouteDeclaration(None=FunctionInput() {inputName=Material, inputType=FunctionInput_MaterialAttributes}) {name=IN_Material, nodeColor=<rgba 0 0.9297 1 1>}
  n2 = NamedRerouteUsage()
  n3 = NamedRerouteUsage()
  FunctionOutput(None=StaticSwitchParameter(True=BlendMaterialAttributes(A=@n1, B=@n2, Alpha=@n3), False=NamedRerouteUsage()) {group=ShoeReactivityTest, parameterName=UseTestReactivity})
  n4 = TextureSampleParameter2D() {
   group=ShoeReactivityTest
   parameterName=ShoeReactivityMask
   samplerType=SAMPLERTYPE_Masks
   texture=&T_BlackMask
  }
  n5 = NamedRerouteDeclaration(None=ChannelMaskParameter(None=@n4.RGB) {parameterName=ShoeReactivityMaskChannel_01}.R) {name=ReactivityMask_01, nodeColor=<rgba 0 0.1328 1 1>}
  n6 = NamedRerouteDeclaration(None=ChannelMaskParameter(None=@n4.RGB) {maskChannel=Green, parameterName=ShoeReactivityMaskChannel_02, defaultValue=<rgba 0 1 0 0>}.R) {name=ReactivityMask_02, nodeColor=<rgba 0 0.1328 1 1>}
  NamedRerouteDeclaration(None=Saturate(None=Add(A=@n5, B=@n6))) {name=ReactivityMask, nodeColor=<rgba 1 0.1875 0 1>}
  n7 = Multiply(A=VectorParameter() {parameterName=ShoeReactivityColor_01, defaultValue=<rgba 1 0 0 1>}.RGB, B=NamedRerouteUsage())
  n8 = Multiply(A=VectorParameter() {parameterName=ShoeReactivityColor_02, defaultValue=<rgba 0 1 0 1>}.RGB, B=NamedRerouteUsage())
  NamedRerouteDeclaration(None=Multiply(A=Add(A=@n7, B=@n8), B=ScalarParameter() {defaultValue=1, parameterName=ShoeReactivityEmissiveStrength})) {name=OUT_Emissive, nodeColor=<rgba 1 0 0.375 1>}
  n9 = MaterialFunctionCall() {materialFunction=/Game/Packages/Fortress_SharedMaterials/Base_Material_Functions/BaseCharacterShaderFunctions/Audio/MF_AudioReactivity}
  n10 = Time()
  n11 = ScalarParameter() {group=ShoeReactivityTest, parameterName=ShoeReactivityElimTime}
  n12 = Saturate(None=Divide(A=Subtract(A=@n10, B=@n11)) {constB=4})
  n13 = Max(
   A=Max(A=Power(Base=@n9.Amplitude) {constExponent=2.5}, B=OneMinus(None=@n12))
   B=ScalarParameter() {group=ShoeReactivityTest, parameterName=ShoeReactivityVelocity}
  )
  NamedRerouteDeclaration(None=Saturate(None=NamedRerouteDeclaration(None=@n13) {name=ReactivityStrength, nodeColor=<rgba 0 1 0.0625 1>})) {name=NormalizedReactivityStrength, nodeColor=<rgba 1 0 0.2578 1>}
  n14 = Constant()
  n15 = SetMaterialAttributes!(
   MaterialAttributes=NamedRerouteUsage()
   BaseColor=VectorParameter() {group=ShoeReactivityTest, parameterName=ShoeReactivityBaseColor, defaultValue=<rgba 1 0.6286 0.1464 1>}.RGB
   EmissiveColor=@n14
   Metallic=ScalarParameter() {defaultValue=1, group=ShoeReactivityTest, parameterName=ShoeReactivityMetalness}
   Roughness=ScalarParameter() {defaultValue=0.2, group=ShoeReactivityTest, parameterName=ShoeReactivityRoughness}
   SubsurfaceColor=@n14
  ) {attributeSetTypes [
    #1
    #2
    #3
    #4
    #5
   ]}
  n16 = NamedRerouteUsage()
  n17 = Constant()
  n18 = NamedRerouteUsage()
  NamedRerouteDeclaration(None=BlendMaterialAttributes(
    A=@n15
    B=SetMaterialAttributes!(MaterialAttributes=@n16, BaseColor=@n17, EmissiveColor=@n18, Metallic=@n17) {attributeSetTypes [#1, #2, #3]}
    Alpha=NamedRerouteUsage()
   )) {name=OUT_Material, nodeColor=<rgba 0 0.5781 1 1>}
 ]}
