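# Material function for hologram support on techart shoe (Kicks) materials that passes material attributes through with no visible change.
# It extracts the Normal, lerps it with itself using the IsKicks scalar parameter (default 1), then writes it back.
# The no-op lerp wires IsKicks into the graph, exposing it under a parameter group marked as not for manual editing.
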MaterialFunction /Game/Global/Materials/Cosmetics/Kicks/Shoes/Materials/MF_TechartShoeHologram
graph {nodes [
  n1 = GetMaterialAttributes(None=FunctionInput() {inputType=FunctionInput_MaterialAttributes}) {attributeGetTypes [#1]}
  n2 = ScalarParameter() {defaultValue=1, group=HOLOGRAMSUPPORTDONOTMODIFYMANUALLY, parameterName=IsKicks}
  FunctionOutput(None=SetMaterialAttributes!(MaterialAttributes=@n1.MaterialAttributes, Normal=LinearInterpolate(A=@n1.Normal, B=@n1.Normal, Alpha=@n2)) {attributeSetTypes [#1]})
 ]}
